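# Physics asset for the M_MED_FaintDoor male character body, centred on its long coat, with capsule bodies for coat, belt, holster, pelvis, legs and an oversized root capsule.
# Five coat bone chains (left, right, three back) use twist-locked, limited-swing constraints between segments.
# Only the back-left and back-right chains are constrained to their base bodies.
# Swing-locked, free-twist cross-links tie the fifth back-side segments to the centre back.
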PhysicsAsset /BRCosmetics/Characters/Player/Male/Medium/Bodies/M_MED_FaintDoor/Meshes/M_MED_FaintDoor_Physics
bodies {
 base_coat_bk [{shapeType=Capsule, length=0, radius=2.6582, shapeName=base_coat_bk_capsule, rotation=<rot 0 -90 0>}]
 base_coat_bk_l [{shapeType=Capsule, length=0, radius=2.6582, shapeName=base_coat_bk_l_capsule, rotation=<rot 0 -90 0>}]
 base_coat_bk_r [{shapeType=Capsule, length=0, radius=2.6582, shapeName=base_coat_bk_r_capsule, rotation=<rot 0 -90 0>}]
 base_coat_l [{shapeType=Capsule, length=0, radius=2.6582, shapeName=base_coat_l_capsule, rotation=<rot 0 -90 0>}]
 base_coat_r [{shapeType=Capsule, length=0, radius=2.6582, shapeName=base_coat_r_capsule, rotation=<rot 0 -90 0>}]
 calf_l [{shapeType=Capsule, length=35.6576, radius=9, center=<-20.7422 0.0237 1.313>, rotation=<rot 0.3148 -90.2014 92.3943>}]
 calf_r [{shapeType=Capsule, length=35.6576, radius=9, center=<20.7422 -0.0237 -1.313>, rotation=<rot 0.3148 -90.2015 -88.6>}]
 dyn_belt_l [{
   shapeType=Capsule
   length=3.4103
   radius=3.5084
   shapeName=dyn_belt_l_capsule
   center=<0.0502 -0.6244 -0.1179>
   rotation=<rot -18.7748 -90 15.8345>
  }]
 dyn_coat_01_bk [{shapeType=Capsule, length=16.8741, radius=8.8219, center=<14.0883 -7.2525 0.2864>}]
 dyn_coat_01_bk_l [{shapeType=Capsule, length=16.8741, radius=8.8219, center=<14.0883 -7.1842 0.2864>}]
 dyn_coat_01_bk_r [{shapeType=Capsule, length=16.8741, radius=8.8219, center=<14.0883 -7.6332 0.2864>}]
 dyn_coat_01_l [{shapeType=Capsule, length=9.3222, radius=8.8219, center=<14.0883 -8.7324 0.2864>, rotation=<rot 0.0001 -1.0363 0>}]
 dyn_coat_01_r [{shapeType=Capsule, length=9.576, radius=8.8219, center=<14.0883 -8.1172 7.6977>, rotation=<rot -0.0001 0 8.2227>}]
 dyn_coat_02_bk [{shapeType=Capsule, length=16.8741, radius=8.8219, center=<14.0883 -1.8975 0>}]
 dyn_coat_02_bk_l [{shapeType=Capsule, length=16.8741, radius=8.8219, center=<14.0883 -5.7534 0.4867>, rotation=<rot 0 0 -12.7878>}]
 dyn_coat_02_bk_r [{shapeType=Capsule, length=16.8741, radius=8.8219, center=<14.0883 -4.3902 0.1614>, rotation=<rot 0 0 11.5187>}]
 dyn_coat_02_l [
  {shapeType=Capsule, length=9.3222, radius=8.8219, center=<14.0883 -8.7488 0.2803>, rotation=<rot 0.0004 0 5.5984>}
  {
   shapeType=Capsule
   length=8.5059
   radius=4.0716
   shapeName=dyn_coat_02_l_capsule
   center=<5.1241 -3.8902 2.8231>
   rotation=<rot -85.2266 -8.4774 9.6559>
  }
 ]
 dyn_coat_02_r [{shapeType=Capsule, length=9.576, radius=8.8219, center=<14.0884 -6.2456 6.9321>, rotation=<rot -0.0012 0.0003 4.3612>}]
 dyn_coat_03_bk [{shapeType=Capsule, length=16.8741, radius=8.8219, center=<14.0883 -4.3931 0>}]
 dyn_coat_03_bk_l [{shapeType=Capsule, length=16.8741, radius=8.8219, center=<14.0883 -4.3931 0.2864>}]
 dyn_coat_03_bk_r [{shapeType=Capsule, length=16.8741, radius=8.8219, center=<14.0883 -4.3931 0.2864>}]
 dyn_coat_03_l [{shapeType=Capsule, length=9.3222, radius=8.8219, center=<14.0883 -6.6152 0.2864>}]
 dyn_coat_03_r [
  {shapeType=Capsule, length=9.576, radius=8.8219, center=<14.0884 -3.3926 5.9601>, rotation=<rot -0.0014 0.0004 10.4686>}
  {
   shapeType=Capsule
   length=7.9963
   radius=5.1489
   shapeName=dyn_coat_03_r_capsule
   center=<9.5314 -1.8811 -1.0789>
   rotation=<rot 83.2391 0.0202 -6.819>
  }
 ]
 dyn_coat_04_bk [{shapeType=Capsule, length=16.8741, radius=8.8219, center=<14.0883 -4.3931 0>}]
 dyn_coat_04_bk_l [{shapeType=Capsule, length=16.8741, radius=8.8219, center=<14.0883 -4.3931 0.2864>}]
 dyn_coat_04_bk_r [{shapeType=Capsule, length=16.8741, radius=8.8219, center=<14.0883 -4.3931 0.2864>}]
 dyn_coat_04_l [{shapeType=Capsule, length=9.3222, radius=8.8219, center=<14.0883 -4.3931 0.2864>}]
 dyn_coat_04_r [{shapeType=Capsule, length=10.8579, radius=8.8219, center=<14.0883 0.1781 4.4855>, rotation=<rot -0.0013 0.0004 12.7704>}]
 dyn_coat_05_bk [{shapeType=Capsule, length=16.8741, radius=8.8219, center=<14.0883 -4.3931 0>}]
 dyn_coat_05_bk_l [{shapeType=Capsule, length=16.8741, radius=8.8219, center=<14.0883 -4.3931 0.2864>}]
 dyn_coat_05_bk_r [{shapeType=Capsule, length=16.8741, radius=8.8219, center=<14.0883 -4.3931 0.2864>}]
 dyn_coat_05_l [{shapeType=Capsule, length=9.3222, radius=8.8219, center=<14.0883 -4.3931 0.2864>}]
 dyn_coat_05_r [{shapeType=Capsule, length=13.2078, radius=8.8219, center=<14.0882 0.7944 1.1869>, rotation=<rot -0.0011 0.0003 8.8532>}]
 dyn_coat_06_bk_l [{shapeType=Capsule, length=12.0798, radius=8.8219, center=<14.0883 -4.3931 0.2864>}]
 dyn_coat_06_bk_r [{shapeType=Capsule, length=11.1604, radius=8.8219, center=<14.0883 -4.3931 0.2864>}]
 dyn_coat_06_l [{shapeType=Capsule, length=9.3222, radius=8.8219, center=<14.0883 -4.3931 0.2864>}]
 dyn_coat_06_r [{shapeType=Capsule, length=11.122, radius=8.8219, center=<14.0882 2.3818 1.7589>, rotation=<rot -0.0011 0.0003 20.4347>}]
 dyn_holster [
  {shapeType=Capsule, length=26.7096, radius=5.6213, center=<8.6966 -4.6926 0.4228>, rotation=<rot 0.1184 93.6625 -90.8307>}
  {
   shapeType=Capsule
   length=21.2384
   radius=3.8757
   shapeName=dyn_holster_capsule
   center=<15.6298 -11.0203 18.5785>
   rotation=<rot 88.1936 9.4052 5.5611>
  }
 ]
 foot_l [{shapeType=Capsule, length=14.7843, radius=10.5874, center=<-0.3294 7.3762 0.9581>, rotation=<rot 0.6652 0.1919 93.5384>}]
 foot_r [{shapeType=Capsule, length=14.7843, radius=10.5874, center=<0.3294 -7.3762 -0.9581>, rotation=<rot 0.6652 0.1919 -86.4616>}]
 pelvis [{shapeType=Capsule, length=4.0014, radius=11.6373, center=<2.5187 0.2195 0.3186>}]
 root [{
   shapeType=Capsule
   length=307.4653
   radius=395.6267
   shapeName=root_capsule
   center=<0.0002 -67.9425 -398.2817>
   rotation=<rot 90 -179.9998 0>
  }]
 thigh_l [
  {shapeType=Capsule, length=30.2991, radius=9.4063, center=<-20.5875 0.3478 1.2552>, rotation=<rot -0.4537 -92.2227 90.5503>}
  {
   shapeType=Capsule
   length=4.0674
   radius=7.2597
   shapeName=thigh_l_capsule
   center=<-15.8918 4.3258 -0.3317>
   rotation=<rot -86.4524 -179.9998 179.9995>
  }
 ]
 thigh_r [{shapeType=Capsule, length=28.0484, radius=9.4063, center=<20.5251 -2.2394 -0.127>, rotation=<rot -0.4537 -92.2227 -89.4497>}]
}
constraints [
 {bone1Name=dyn_coat_04_l, bone2Name=dyn_coat_03_l, swing1LimitDegrees=15, swing2LimitDegrees=6, twistLimitDegrees=45, twistMotion=Locked}
 {
  bone1Name=dyn_coat_02_bk_l
  bone2Name=dyn_coat_01_bk_l
  swing1LimitDegrees=15
  swing2LimitDegrees=6
  twistLimitDegrees=45
  twistMotion=Locked
 }
 {
  bone1Name=dyn_coat_05_bk_l
  bone2Name=dyn_coat_04_bk_l
  swing1LimitDegrees=15
  swing2LimitDegrees=6
  twistLimitDegrees=45
  twistMotion=Locked
 }
 {bone1Name=dyn_coat_04_bk, bone2Name=dyn_coat_03_bk, swing1LimitDegrees=15, swing2LimitDegrees=6, twistLimitDegrees=45, twistMotion=Locked}
 {
  bone1Name=dyn_coat_02_bk_r
  bone2Name=dyn_coat_01_bk_r
  swing1LimitDegrees=15
  swing2LimitDegrees=6
  twistLimitDegrees=45
  twistMotion=Locked
 }
 {
  bone1Name=dyn_coat_05_bk_r
  bone2Name=dyn_coat_04_bk_r
  swing1LimitDegrees=15
  swing2LimitDegrees=6
  twistLimitDegrees=45
  twistMotion=Locked
 }
 {bone1Name=dyn_coat_04_r, bone2Name=dyn_coat_03_r, swing1LimitDegrees=15, swing2LimitDegrees=15, twistLimitDegrees=45, twistMotion=Locked}
 {bone1Name=dyn_coat_02_bk, bone2Name=dyn_coat_01_bk, swing1LimitDegrees=15, swing2LimitDegrees=6, twistLimitDegrees=45, twistMotion=Locked}
 {bone1Name=dyn_coat_03_bk, bone2Name=dyn_coat_02_bk, swing1LimitDegrees=15, swing2LimitDegrees=6, twistLimitDegrees=45, twistMotion=Locked}
 {bone1Name=dyn_coat_05_bk, bone2Name=dyn_coat_04_bk, swing1LimitDegrees=15, swing2LimitDegrees=6, twistLimitDegrees=45, twistMotion=Locked}
 {bone1Name=dyn_coat_01_bk_l, bone2Name=base_coat_bk_l, swing1LimitDegrees=15, swing2LimitDegrees=6, twistLimitDegrees=45, twistMotion=Locked}
 {
  bone1Name=dyn_coat_03_bk_l
  bone2Name=dyn_coat_02_bk_l
  swing1LimitDegrees=15
  swing2LimitDegrees=6
  twistLimitDegrees=45
  twistMotion=Locked
 }
 {
  bone1Name=dyn_coat_04_bk_l
  bone2Name=dyn_coat_03_bk_l
  swing1LimitDegrees=15
  swing2LimitDegrees=6
  twistLimitDegrees=45
  twistMotion=Locked
 }
 {
  bone1Name=dyn_coat_06_bk_l
  bone2Name=dyn_coat_05_bk_l
  swing1LimitDegrees=15
  swing2LimitDegrees=6
  twistLimitDegrees=45
  twistMotion=Locked
 }
 {bone1Name=dyn_coat_01_bk_r, bone2Name=base_coat_bk_r, swing1LimitDegrees=15, swing2LimitDegrees=6, twistLimitDegrees=45, twistMotion=Locked}
 {
  bone1Name=dyn_coat_03_bk_r
  bone2Name=dyn_coat_02_bk_r
  swing1LimitDegrees=15
  swing2LimitDegrees=6
  twistLimitDegrees=45
  twistMotion=Locked
 }
 {
  bone1Name=dyn_coat_04_bk_r
  bone2Name=dyn_coat_03_bk_r
  swing1LimitDegrees=15
  swing2LimitDegrees=6
  twistLimitDegrees=45
  twistMotion=Locked
 }
 {
  bone1Name=dyn_coat_06_bk_r
  bone2Name=dyn_coat_05_bk_r
  swing1LimitDegrees=15
  swing2LimitDegrees=6
  twistLimitDegrees=45
  twistMotion=Locked
 }
 {bone1Name=dyn_coat_02_l, bone2Name=dyn_coat_01_l, swing1LimitDegrees=15, swing2LimitDegrees=6, twistLimitDegrees=45, twistMotion=Locked}
 {bone1Name=dyn_coat_03_l, bone2Name=dyn_coat_02_l, swing1LimitDegrees=15, swing2LimitDegrees=6, twistLimitDegrees=45, twistMotion=Locked}
 {bone1Name=dyn_coat_05_l, bone2Name=dyn_coat_04_l, swing1LimitDegrees=15, swing2LimitDegrees=6, twistLimitDegrees=45, twistMotion=Locked}
 {bone1Name=dyn_coat_06_l, bone2Name=dyn_coat_05_l, swing1LimitDegrees=15, swing2LimitDegrees=6, twistLimitDegrees=45, twistMotion=Locked}
 {bone1Name=dyn_coat_02_r, bone2Name=dyn_coat_01_r, swing1LimitDegrees=15, swing2LimitDegrees=15, twistLimitDegrees=45, twistMotion=Locked}
 {bone1Name=dyn_coat_03_r, bone2Name=dyn_coat_02_r, swing1LimitDegrees=15, swing2LimitDegrees=15, twistLimitDegrees=45, twistMotion=Locked}
 {bone1Name=dyn_coat_05_r, bone2Name=dyn_coat_04_r, swing1LimitDegrees=15, swing2LimitDegrees=15, twistLimitDegrees=45, twistMotion=Locked}
 {bone1Name=dyn_coat_06_r, bone2Name=dyn_coat_05_r, swing1LimitDegrees=15, swing2LimitDegrees=15, twistLimitDegrees=45, twistMotion=Locked}
 {
  bone1Name=dyn_coat_05_bk_r
  bone2Name=dyn_coat_05_bk
  swing1LimitDegrees=45
  swing1Motion=Locked
  swing2LimitDegrees=45
  swing2Motion=Locked
  twistLimitDegrees=45
  twistMotion=Free
 }
 {
  bone1Name=dyn_coat_05_bk_l
  bone2Name=dyn_coat_05_bk
  swing1LimitDegrees=45
  swing1Motion=Locked
  swing2LimitDegrees=45
  swing2Motion=Locked
  twistLimitDegrees=45
  twistMotion=Free
 }
]
