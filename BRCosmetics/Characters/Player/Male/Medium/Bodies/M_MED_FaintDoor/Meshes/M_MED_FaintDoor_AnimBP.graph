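# Animation Blueprint for the M_MED_FaintDoor cosmetic player body, adding secondary motion to its holster, cape, collar, straps and other parts.
# It copies the parent mesh's pose into a Control Rig driven by movement, emote and hat-variant flags.
# A RigidBody node, chained Anim Dynamics springs and small corrective bone transforms follow.
# Each update sets a gravity override from a parameter asset and HatVariant from the Stage1 cosmetic tag.
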AnimBlueprint /BRCosmetics/Characters/Player/Male/Medium/Bodies/M_MED_FaintDoor/Meshes/M_MED_FaintDoor_AnimBP
target_skeleton=&/M_MED_FaintDoor_Skeleton
graphs [
 AnimGraph {nodes [
   n1 = Control Rig!(
    Source=Copy Pose From Mesh() {bUseAttachedParent=true}
    Alpha
    isCrouching_CR:bool=Get bIsCrouching()
    isKneeSliding_CR:bool=Get bIsKneeSliding()
    isEmoting_CR:bool=Get bIsPlayingEmote()
    isSkydiving_CR:bool=Get bIsSkyDiving()
    isBaseballSliding_CR:bool=Get bIsBaseballSliding()
    isWearingHat:bool=Get HatVariant() -> (HatVariant:bool)
   ) -> (Pose) {controlRigAssetReference {controlRigAsset=&/M_MED_FaintDoor_CtrlRig}}
   n2 = Anim Dynamics - Bone: dyn_holster!(ComponentPose=Local To Component(LocalPose=@n1), Alpha=1, GravityOverride:Vector="0,0,-980", ExternalForce="0,0,0") -> (Pose) {
    angularDampingOverride=1
    angularSpringConstant=17000
    bAngularSpring=true
    bOverrideAngularDamping=true
    bUseGravityOverride=true
    gravityScale=0.35
    simulationSpace=RootRelative
    boundBone {boneName=dyn_holster}
    componentAppliedLinearAccClamp=<3000 3000 3000>
    componentLinearAccScale=<0.3 0.3 0.3>
    componentLinearVelScale=<0.3 0.3 0.3>
    gravityOverride {z=-980}
    physicsBodyDefinitions [{
      boundBone {boneName=dyn_holster}
      boxExtents=<35 35 35>
      constraintSetup {angularLimitsMax=<0 0 6>, angularLimitsMin=<0 -5 -6>, angularTarget=<1 0 0>}
      localJointOffset=<2 0 0>
     }]
    simSpaceSettings {simSpaceAngularAlpha=0.15}
   }
   n3 = RigidBody!(ComponentPose=@n2, Alpha, OverrideWorldGravity:Vector="0,0,0", ExternalForce) -> (Pose) {
    bOverrideWorldGravity=true
    overridePhysicsAsset=&/M_MED_FaintDoor_Physics
    simulationSpace=BaseBoneSpace
    baseBoneRef {boneName=pelvis}
    simSpaceSettings {
     maxAngularAcceleration=45
     maxAngularVelocity=8
     maxLinearAcceleration=1000
     maxLinearVelocity=300
     worldAlpha=1
     externalLinearDragV=<1 1 0.8>
    }
   }
   n4 = "Transform (Modify) Bone - Bone: base_upperBeltVolume"(ComponentPose=@n3, Alpha=1, Translation="0,0,0", Rotation="0,0,0", Scale="1,0.96,0.96") {scaleMode=BMM_Replace, scaleSpace=BCS_BoneSpace, boneToModify {boneName=base_upperBeltVolume}, scale {y=0.96, z=0.96}}
   n5 = Anim Dynamics - Bone: dyn_shoulderCape_l!(ComponentPose=@n4, Alpha=0.6, GravityOverride:Vector="0,0,-980", ExternalForce="0,0,0") -> (Pose) {
    angularDampingOverride=1
    angularSpringConstant=10000
    bAngularSpring=true
    bOverrideAngularDamping=true
    bUseGravityOverride=true
    gravityScale=0.35
    simulationSpace=BoneRelative
    alphaScaleBiasClamp {bInterpResult=true, interpSpeedDecreasing=5}
    boundBone {boneName=dyn_shoulderCape_l}
    componentAppliedLinearAccClamp=<3000 3000 3000>
    componentLinearAccScale=<0.3 0.3 0.3>
    componentLinearVelScale=<0.3 0.4 3>
    gravityOverride {z=-980}
    physicsBodyDefinitions [{
      boundBone {boneName=dyn_shoulderCape_l}
      boxExtents=<32 32 32>
      constraintSetup {angularLimitsMax=<0 6 0>, angularLimitsMin=<0 -6 -60>, angularTarget=<1 0 0>}
      localJointOffset=<0.5 0 0>
     }]
    relativeSpaceBone {boneName=root}
    simSpaceSettings {simSpaceAngularAlpha=0.2}
   }
   n6 = Anim Dynamics - Bone: dyn_shoulderCape_fr_l!(ComponentPose=@n5, Alpha=1, GravityOverride:Vector="0,0,-980", ExternalForce="0,0,0") -> (Pose) {
    angularDampingOverride=1
    angularSpringConstant=8000
    bAngularSpring=true
    bOverrideAngularDamping=true
    bUseGravityOverride=true
    gravityScale=0.35
    simulationSpace=BoneRelative
    alphaScaleBiasClamp {bInterpResult=true, interpSpeedDecreasing=5}
    boundBone {boneName=dyn_shoulderCape_fr_l}
    componentAppliedLinearAccClamp=<4000 4000 4000>
    componentLinearAccScale=<0.3 0.3 0.3>
    componentLinearVelScale=<0.3 0.4 3>
    gravityOverride {z=-980}
    physicsBodyDefinitions [{
      boundBone {boneName=dyn_shoulderCape_fr_l}
      boxExtents=<25 25 25>
      constraintSetup {angularLimitsMax=<0 10 0>, angularLimitsMin=<0 -10 -60>, angularTarget=<1 0 0>}
      localJointOffset=<1 0 0>
     }]
    relativeSpaceBone {boneName=root}
    simSpaceSettings {simSpaceAngularAlpha=0.4}
   }
   n7 = Anim Dynamics - Bone: dyn_shoulderCape_bk_l!(ComponentPose=@n6, Alpha=1, GravityOverride:Vector="0,0,-980", ExternalForce="0,0,0") -> (Pose) {
    angularDampingOverride=1
    angularSpringConstant=8000
    bAngularSpring=true
    bOverrideAngularDamping=true
    bUseGravityOverride=true
    gravityScale=0.35
    simulationSpace=BoneRelative
    alphaScaleBiasClamp {bInterpResult=true, interpSpeedDecreasing=5}
    boundBone {boneName=dyn_shoulderCape_bk_l}
    componentAppliedLinearAccClamp=<4000 4000 4000>
    componentLinearAccScale=<0.3 0.3 0.3>
    componentLinearVelScale=<0.3 0.4 3>
    gravityOverride {z=-980}
    physicsBodyDefinitions [{
      boundBone {boneName=dyn_shoulderCape_bk_l}
      boxExtents=<25 25 25>
      constraintSetup {angularLimitsMax=<0 10 0>, angularLimitsMin=<0 -10 -60>, angularTarget=<1 0 0>}
      localJointOffset=<1 0 0>
     }]
    relativeSpaceBone {boneName=root}
    simSpaceSettings {simSpaceAngularAlpha=0.4}
   }
   n8 = Anim Dynamics - Bone: dyn_shoulderCape_fr!(ComponentPose=@n7, Alpha=1, GravityOverride:Vector="0,0,-980", ExternalForce="0,0,0") -> (Pose) {
    angularDampingOverride=1
    angularSpringConstant=8000
    bAngularSpring=true
    bOverrideAngularDamping=true
    bUseGravityOverride=true
    gravityScale=0.35
    simulationSpace=BoneRelative
    alphaScaleBiasClamp {bInterpResult=true, interpSpeedDecreasing=5}
    boundBone {boneName=dyn_shoulderCape_fr}
    componentAppliedLinearAccClamp=<4000 4000 4000>
    componentLinearAccScale=<0.3 0.3 0.3>
    componentLinearVelScale=<0.3 0.4 3>
    gravityOverride {z=-980}
    physicsBodyDefinitions [{
      boundBone {boneName=dyn_shoulderCape_fr}
      boxExtents=<25 25 25>
      constraintSetup {angularLimitsMax=<0 10 0>, angularLimitsMin=<0 -10 -60>, angularTarget=<1 0 0>}
      localJointOffset=<1 0 0>
     }]
    relativeSpaceBone {boneName=root}
    simSpaceSettings {simSpaceAngularAlpha=0.4}
   }
   n9 = Anim Dynamics - Bone: dyn_shoulderCape_bk!(ComponentPose=@n8, Alpha=1, GravityOverride:Vector="0,0,-980", ExternalForce="0,0,0") -> (Pose) {
    angularDampingOverride=1
    angularSpringConstant=8000
    bAngularSpring=true
    bOverrideAngularDamping=true
    bUseGravityOverride=true
    gravityScale=0.35
    simulationSpace=BoneRelative
    alphaScaleBiasClamp {bInterpResult=true, interpSpeedDecreasing=5}
    boundBone {boneName=dyn_shoulderCape_bk}
    componentAppliedLinearAccClamp=<4000 4000 4000>
    componentLinearAccScale=<0.3 0.3 0.3>
    componentLinearVelScale=<0.3 0.4 3>
    gravityOverride {z=-980}
    physicsBodyDefinitions [{
      boundBone {boneName=dyn_shoulderCape_bk}
      boxExtents=<22 22 22>
      constraintSetup {angularLimitsMax=<0 10 0>, angularLimitsMin=<0 -10 -60>, angularTarget=<1 0 0>}
      localJointOffset=<1 0 0>
     }]
    relativeSpaceBone {boneName=root}
    simSpaceSettings {simSpaceAngularAlpha=0.4}
   }
   n10 = Anim Dynamics - Bone: dyn_collar_01_bk!(ComponentPose=@n9, Alpha=0.5, GravityOverride:Vector="0,0,-980", ExternalForce="0,0,0") -> (Pose) {
    angularDampingOverride=1
    angularSpringConstant=8000
    bAngularSpring=true
    bOverrideAngularDamping=true
    bUseGravityOverride=true
    gravityScale=0.35
    simulationSpace=BoneRelative
    alphaScaleBiasClamp {bInterpResult=true, interpSpeedDecreasing=5}
    boundBone {boneName=dyn_collar_01_bk}
    componentAppliedLinearAccClamp=<4000 4000 4000>
    componentLinearAccScale=<0.4 0.4 0.4>
    componentLinearVelScale=<0.4 0.4 0.4>
    gravityOverride {z=-980}
    physicsBodyDefinitions [{
      boundBone {boneName=dyn_collar_01_bk}
      boxExtents=<26 26 26>
      constraintSetup {angularLimitsMax=<0 10 0>, angularLimitsMin=<0 -10 -20>, angularTarget=<1 0 0>}
      localJointOffset=<-2 0 0>
     }]
    relativeSpaceBone {boneName=root}
    simSpaceSettings {simSpaceAngularAlpha=0.2}
   }
   n11 = Anim Dynamics - Bone: dyn_collar_bk_l!(ComponentPose=@n10, Alpha=1, GravityOverride:Vector="0,0,-980", ExternalForce="0,0,0") -> (Pose) {
    angularDampingOverride=1
    angularSpringConstant=8000
    bAngularSpring=true
    bOverrideAngularDamping=true
    bUseGravityOverride=true
    gravityScale=0.35
    simulationSpace=BoneRelative
    alphaScaleBiasClamp {bInterpResult=true, interpSpeedDecreasing=5}
    boundBone {boneName=dyn_collar_bk_l}
    componentAppliedLinearAccClamp=<4000 4000 4000>
    componentLinearAccScale=<0.4 0.4 0.4>
    componentLinearVelScale=<0.4 0.4 0.4>
    gravityOverride {z=-980}
    physicsBodyDefinitions [{
      boundBone {boneName=dyn_collar_bk_l}
      boxExtents=<26 26 26>
      constraintSetup {angularLimitsMax=<0 10 0>, angularLimitsMin=<0 -10 -30>, angularTarget=<1 0 0>}
      localJointOffset=<-2 0 0>
     }]
    relativeSpaceBone {boneName=root}
    simSpaceSettings {simSpaceAngularAlpha=0.2}
   }
   n12 = Anim Dynamics - Bone: dyn_collar_01_fr!(ComponentPose=@n11, Alpha=0.5, GravityOverride:Vector="0,0,-980", ExternalForce="0,0,0") -> (Pose) {
    angularDampingOverride=1
    angularSpringConstant=8000
    bAngularSpring=true
    bOverrideAngularDamping=true
    bUseGravityOverride=true
    gravityScale=0.35
    simulationSpace=BoneRelative
    alphaScaleBiasClamp {bInterpResult=true, interpSpeedDecreasing=5}
    boundBone {boneName=dyn_collar_01_fr}
    componentAppliedLinearAccClamp=<4000 4000 4000>
    componentLinearAccScale=<0.4 0.4 0.4>
    componentLinearVelScale=<0.4 0.4 0.4>
    gravityOverride {z=-980}
    physicsBodyDefinitions [{
      boundBone {boneName=dyn_collar_01_fr}
      boxExtents=<26 26 26>
      constraintSetup {angularLimitsMax=<0 10 0>, angularLimitsMin=<0 -10 -20>, angularTarget=<1 0 0>}
      localJointOffset=<-2 0 0>
     }]
    relativeSpaceBone {boneName=root}
    simSpaceSettings {simSpaceAngularAlpha=0.2}
   }
   n13 = Anim Dynamics - Bone: dyn_collar_fr_l!(ComponentPose=@n12, Alpha=1, GravityOverride:Vector="0,0,-980", ExternalForce="0,0,0") -> (Pose) {
    angularDampingOverride=1
    angularSpringConstant=8000
    bAngularSpring=true
    bOverrideAngularDamping=true
    bUseGravityOverride=true
    gravityScale=0.35
    simulationSpace=BoneRelative
    alphaScaleBiasClamp {bInterpResult=true, interpSpeedDecreasing=5}
    boundBone {boneName=dyn_collar_fr_l}
    componentAppliedLinearAccClamp=<4000 4000 4000>
    componentLinearAccScale=<0.4 0.4 0.4>
    componentLinearVelScale=<0.4 0.4 0.4>
    gravityOverride {z=-980}
    physicsBodyDefinitions [{
      boundBone {boneName=dyn_collar_fr_l}
      boxExtents=<26 26 26>
      constraintSetup {angularLimitsMax=<0 10 0>, angularLimitsMin=<0 -10 -30>, angularTarget=<1 0 0>}
      localJointOffset=<-2 0 0>
     }]
    relativeSpaceBone {boneName=root}
    simSpaceSettings {simSpaceAngularAlpha=0.2}
   }
   n14 = Anim Dynamics - Bone: dyn_collar_l!(ComponentPose=@n13, Alpha=1, GravityOverride:Vector="0,0,-980", ExternalForce="0,0,0") -> (Pose) {
    angularDampingOverride=1
    angularSpringConstant=8000
    bAngularSpring=true
    bOverrideAngularDamping=true
    bUseGravityOverride=true
    gravityScale=0.35
    simulationSpace=BoneRelative
    alphaScaleBiasClamp {bInterpResult=true, interpSpeedDecreasing=5}
    boundBone {boneName=dyn_collar_l}
    componentAppliedLinearAccClamp=<4000 4000 4000>
    componentLinearAccScale=<0.4 0.4 0.4>
    componentLinearVelScale=<0.4 0.4 0.4>
    gravityOverride {z=-980}
    physicsBodyDefinitions [{
      boundBone {boneName=dyn_collar_l}
      boxExtents=<26 26 26>
      constraintSetup {angularLimitsMax=<0 10 0>, angularLimitsMin=<0 -10 -30>, angularTarget=<1 0 0>}
      localJointOffset=<-2 0 0>
     }]
    relativeSpaceBone {boneName=root}
    simSpaceSettings {simSpaceAngularAlpha=0.2}
   }
   n15 = Anim Dynamics - Bone: dyn_collar_02_bk!(ComponentPose=@n14, Alpha=0.5, GravityOverride:Vector="0,0,-980", ExternalForce="0,0,0") -> (Pose) {
    angularDampingOverride=1
    angularSpringConstant=8000
    bAngularSpring=true
    bOverrideAngularDamping=true
    bUseGravityOverride=true
    gravityScale=0.35
    simulationSpace=BoneRelative
    alphaScaleBiasClamp {bInterpResult=true, interpSpeedDecreasing=5}
    boundBone {boneName=dyn_collar_02_bk}
    componentAppliedLinearAccClamp=<4000 4000 4000>
    componentLinearAccScale=<0.2 0.2 0.2>
    componentLinearVelScale=<0.2 0.2 0.2>
    gravityOverride {z=-980}
    physicsBodyDefinitions [{
      boundBone {boneName=dyn_collar_02_bk}
      boxExtents=<35 35 35>
      constraintSetup {angularLimitsMax=<0 10 0>, angularLimitsMin=<0 -10 -10>, angularTarget=<1 0 0>}
      localJointOffset=<-2 0 0>
     }]
    relativeSpaceBone {boneName=root}
    simSpaceSettings {simSpaceAngularAlpha=0.2}
   }
   n16 = Anim Dynamics - Bone: dyn_collar_02_fr!(ComponentPose=@n15, Alpha=0.5, GravityOverride:Vector="0,0,-980", ExternalForce="0,0,0") -> (Pose) {
    angularDampingOverride=1
    angularSpringConstant=8000
    bAngularSpring=true
    bOverrideAngularDamping=true
    bUseGravityOverride=true
    gravityScale=0.35
    simulationSpace=BoneRelative
    alphaScaleBiasClamp {bInterpResult=true, interpSpeedDecreasing=5}
    boundBone {boneName=dyn_collar_02_fr}
    componentAppliedLinearAccClamp=<4000 4000 4000>
    componentLinearAccScale=<0.2 0.2 0.2>
    componentLinearVelScale=<0.2 0.2 0.2>
    gravityOverride {z=-980}
    physicsBodyDefinitions [{
      boundBone {boneName=dyn_collar_02_fr}
      boxExtents=<35 35 35>
      constraintSetup {angularLimitsMax=<0 10 0>, angularLimitsMin=<0 -10 -10>, angularTarget=<1 0 0>}
      localJointOffset=<-2 0 0>
     }]
    relativeSpaceBone {boneName=root}
    simSpaceSettings {simSpaceAngularAlpha=0.2}
   }
   n17 = Anim Dynamics - Bone: dyn_armStraps!(ComponentPose=@n16, Alpha=0.6, GravityOverride:Vector="0,0,-980", ExternalForce="0,0,0") -> (Pose) {
    angularDampingOverride=1
    angularSpringConstant=1000
    bAngularSpring=true
    bOverrideAngularDamping=true
    bUseGravityOverride=true
    gravityScale=0.35
    simulationSpace=BoneRelative
    alphaScaleBiasClamp {bInterpResult=true, interpSpeedDecreasing=5}
    boundBone {boneName=dyn_armStraps}
    componentAppliedLinearAccClamp=<4000 4000 4000>
    componentLinearAccScale=<0.4 0.4 0.5>
    componentLinearVelScale=<0.4 0.4 0.5>
    gravityOverride {z=-980}
    physicsBodyDefinitions [{
      boundBone {boneName=dyn_armStraps}
      boxExtents=<22 22 22>
      constraintSetup {angularLimitsMax=<0 8 0>, angularLimitsMin=<0 -8 -25>, angularTarget=<1 0 0>}
      localJointOffset=<2 -1 -1.5>
     }]
    relativeSpaceBone {boneName=root}
    simSpaceSettings {simSpaceAngularAlpha=0.2}
   }
   n18 = "Transform (Modify) Bone - Bone: dyn_coat_01_l"!(ComponentPose=@n17, bAlphaBoolEnabled:bool=True, Translation="0,-0.7,0", Rotation="-0.412,-3.185,-0.238", Scale="1,1,1") -> (Pose) {
    alphaInputType=Bool
    rotationMode=BMM_Additive
    rotationSpace=BCS_BoneSpace
    scaleSpace=BCS_BoneSpace
    translationMode=BMM_Additive
    translationSpace=BCS_BoneSpace
    alphaBoolBlend {blendInTime=1, blendOutTime=1}
    boneToModify {boneName=dyn_coat_01_l}
    rotation=<rot -0.412 -3.185 -0.238>
    translation {y=-0.7}
   }
   Output Pose(Result=Component To Local(ComponentPose=@n18)) {layerGroup=None}
  ]}
 EventGraph {nodes [
   Event BlueprintUpdateAnimation()
   > n1 = "Gravity Override State Function (Parameter Asset) (Nativized)"(
    CharacterPartAnimInstance=Self-Reference()
    GameTimeInSeconds=GetGameTimeInSeconds()
    DebugJointName=spine_05
    Parameters=&/M_MED_FaintDoor_Parameters
   )
   > Set Gravity Override(Gravity Override=@n1)
   > n2 = Branch(Condition=CharacterItemDefinitionHasAnyMetaTag(Tags=Make Array("[0]"=(TagName="Cosmetics.Variant.Property.Stage1"))))
   > Set HatVariant(HatVariant:bool=true) -> (Output_Get:bool)
   Set HatVariant(execute=@n2.else, HatVariant=false)
  ]}
]
variables [
 Gravity Override {type=Vector}
 HatVariant {type=boolean}
]
