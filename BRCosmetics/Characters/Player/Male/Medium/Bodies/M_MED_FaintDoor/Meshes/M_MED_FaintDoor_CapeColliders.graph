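# Physics asset of cape colliders for the M_MED_FaintDoor male medium player body, apparently so a simulated cape rests on the body and coat.
# A head sphere and capsules on neck, clavicles, pelvis and spine cover the upper body.
# Repeated offset capsules follow the dynamic coat bone chains around the back and sides, mostly narrower at the lowest row, which lacks a center-back body.
# Arm and leg bodies carry no shapes, and there are no constraints.
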PhysicsAsset /BRCosmetics/Characters/Player/Male/Medium/Bodies/M_MED_FaintDoor/Meshes/M_MED_FaintDoor_CapeColliders
bodies {
 calf_l []
 calf_r []
 clavicle_l [{shapeType=Capsule, length=9.0192, radius=9.8871, center=<9.4266 -1.1806 -2.4977>, rotation=<rot 69.3297 1.4476 176.9462>}]
 clavicle_r [{shapeType=Capsule, length=9.0192, radius=9.8871, center=<-9.4266 1.1806 2.4977>, rotation=<rot 69.3297 1.4476 -3.0538>}]
 dyn_coat_02_bk [{
   shapeType=Capsule
   length=12.1805
   radius=13.0269
   shapeName=dyn_coat_02_bk_capsule
   center=<11.7538 12.9065 0.5362>
   rotation=<rot 0 90 -180>
  }]
 dyn_coat_02_bk_l [{
   shapeType=Capsule
   length=12.1805
   radius=13.0269
   shapeName=dyn_coat_02_bk_capsule
   center=<11.7538 12.1683 0.8514>
   rotation=<rot 18.0553 90.0001 -179.9999>
  }]
 dyn_coat_02_bk_r [{
   shapeType=Capsule
   length=12.1805
   radius=13.0269
   shapeName=dyn_coat_02_bk_capsule
   center=<11.7538 11.1793 0.5522>
   rotation=<rot -36.2704 90 180>
  }]
 dyn_coat_02_l [{
   shapeType=Capsule
   length=12.1805
   radius=13.0269
   shapeName=dyn_coat_02_bk_capsule
   center=<11.7538 11.2012 0.5362>
   rotation=<rot 17.4343 90 -179.9999>
  }]
 dyn_coat_02_r [{
   shapeType=Capsule
   length=12.1805
   radius=13.0269
   shapeName=dyn_coat_02_bk_capsule
   center=<11.7538 11.2012 0.5362>
   rotation=<rot -30.1489 90 179.9999>
  }]
 dyn_coat_03_bk [{
   shapeType=Capsule
   length=12.1805
   radius=13.0269
   shapeName=dyn_coat_02_bk_capsule
   center=<11.7538 12.7707 0.5362>
   rotation=<rot 0 90 -180>
  }]
 dyn_coat_03_bk_l [{
   shapeType=Capsule
   length=12.1805
   radius=13.0269
   shapeName=dyn_coat_02_bk_capsule
   center=<11.7538 12.0325 0.8514>
   rotation=<rot 18.0553 90.0001 -179.9999>
  }]
 dyn_coat_03_bk_r [{
   shapeType=Capsule
   length=12.1805
   radius=13.0269
   shapeName=dyn_coat_02_bk_capsule
   center=<11.7538 11.0435 0.5522>
   rotation=<rot -36.2704 90 180>
  }]
 dyn_coat_03_l [{
   shapeType=Capsule
   length=12.1805
   radius=13.0269
   shapeName=dyn_coat_02_bk_capsule
   center=<11.7538 11.0654 0.5362>
   rotation=<rot 17.4343 90 -179.9999>
  }]
 dyn_coat_03_r [{
   shapeType=Capsule
   length=12.1805
   radius=13.0269
   shapeName=dyn_coat_02_bk_capsule
   center=<11.7538 11.0654 0.5362>
   rotation=<rot -30.1489 90 179.9999>
  }]
 dyn_coat_04_bk [{
   shapeType=Capsule
   length=12.1805
   radius=13.0269
   shapeName=dyn_coat_02_bk_capsule
   center=<11.7538 11.8916 0.5362>
   rotation=<rot 0 90 -180>
  }]
 dyn_coat_04_bk_l [{
   shapeType=Capsule
   length=12.1805
   radius=13.0269
   shapeName=dyn_coat_02_bk_capsule
   center=<11.7538 11.1534 0.8514>
   rotation=<rot 18.0553 90.0001 -179.9999>
  }]
 dyn_coat_04_bk_r [{
   shapeType=Capsule
   length=12.1805
   radius=13.0269
   shapeName=dyn_coat_02_bk_capsule
   center=<11.7538 10.1644 0.5522>
   rotation=<rot -36.2704 90 180>
  }]
 dyn_coat_04_l [{
   shapeType=Capsule
   length=12.1805
   radius=13.0269
   shapeName=dyn_coat_02_bk_capsule
   center=<11.7538 10.1863 0.5362>
   rotation=<rot 17.4343 90 -179.9999>
  }]
 dyn_coat_04_r [{
   shapeType=Capsule
   length=12.1805
   radius=13.0269
   shapeName=dyn_coat_02_bk_capsule
   center=<11.7538 10.1863 0.5362>
   rotation=<rot -30.1489 90 179.9999>
  }]
 dyn_coat_05_bk [{
   shapeType=Capsule
   length=12.1805
   radius=13.0269
   shapeName=dyn_coat_02_bk_capsule
   center=<11.7538 11.0224 0.5362>
   rotation=<rot 0 90 -180>
  }]
 dyn_coat_05_bk_l [{
   shapeType=Capsule
   length=12.1805
   radius=13.0269
   shapeName=dyn_coat_02_bk_capsule
   center=<11.7538 10.2841 0.8514>
   rotation=<rot 18.0553 90.0001 -179.9999>
  }]
 dyn_coat_05_bk_r [{
   shapeType=Capsule
   length=12.1805
   radius=13.0269
   shapeName=dyn_coat_02_bk_capsule
   center=<11.7538 9.2952 0.5522>
   rotation=<rot -36.2704 90 180>
  }]
 dyn_coat_05_l [{
   shapeType=Capsule
   length=12.1805
   radius=13.0269
   shapeName=dyn_coat_02_bk_capsule
   center=<11.7538 9.317 0.5362>
   rotation=<rot 17.4343 90 -179.9999>
  }]
 dyn_coat_05_r [{
   shapeType=Capsule
   length=12.1805
   radius=13.0269
   shapeName=dyn_coat_02_bk_capsule
   center=<11.7538 9.317 0.5362>
   rotation=<rot -30.1489 90 179.9999>
  }]
 dyn_coat_06_bk_l [{
   shapeType=Capsule
   length=12.1805
   radius=9.3684
   shapeName=dyn_coat_02_bk_capsule
   center=<13.6867 4.659 -0.3408>
   rotation=<rot 29.0453 90.0001 -179.9999>
  }]
 dyn_coat_06_bk_r [{
   shapeType=Capsule
   length=12.1805
   radius=9.3684
   shapeName=dyn_coat_02_bk_capsule
   center=<12.8557 6.9312 -0.2655>
   rotation=<rot -36.2704 90 180>
  }]
 dyn_coat_06_l [{
   shapeType=Capsule
   length=12.1805
   radius=13.0269
   shapeName=dyn_coat_02_bk_capsule
   center=<11.7538 5.8387 0.5362>
   rotation=<rot 17.4343 90 -179.9999>
  }]
 dyn_coat_06_r [{
   shapeType=Capsule
   length=12.1805
   radius=10.2401
   shapeName=dyn_coat_02_bk_capsule
   center=<11.7538 5.8135 0.8704>
   rotation=<rot -43.3535 90.0001 179.9999>
  }]
 foot_l []
 foot_r []
 head [{shapeType=Sphere, radius=17.3125, center=<11.028 -0.174 0>}]
 lowerarm_l []
 lowerarm_r []
 neck_02 [{shapeType=Capsule, length=13.4375, radius=7.3175, center=<-0.2642 -0.3408 0>, rotation=<rot 90 -171.1193 -175.3978>}]
 pelvis [{shapeType=Capsule, length=8.6456, radius=14.5388, center=<-2.4887 0.1881 0>}]
 spine_01 [{shapeType=Capsule, length=7.9438, radius=13.5586, center=<-1.7965 0.1884 0>}]
 spine_02 [
  {shapeType=Capsule, length=7.3264, radius=13.6322, center=<-1.711 -0.1851 0.0001>}
  {shapeType=Capsule, length=29.3956, radius=11.5807, center=<2.5024 0.4992 3.9825>, rotation=<rot 90 0.0049 0.0049>}
  {shapeType=Capsule, length=29.875, radius=11.1013, center=<2.4704 0.4991 -3.9825>, rotation=<rot 90 -0.0396 -0.0396>}
 ]
 spine_03 [{shapeType=Capsule, length=8.2107, radius=13.7031, center=<0.1445 0.2396 0.0001>}]
 spine_04 [{shapeType=Capsule, length=9.4404, radius=14.1325, center=<2.766 0.7765 0.0002>}]
 spine_05 [
  {shapeType=Capsule, length=8.0226, radius=15.695, center=<-2.2795 1.8109 0>}
  {shapeType=Capsule, length=13.9518, radius=13.9425, center=<-0.2916 0.4452 0>}
 ]
 thigh_l []
 thigh_r []
 upperarm_l []
 upperarm_r []
}
constraints []
