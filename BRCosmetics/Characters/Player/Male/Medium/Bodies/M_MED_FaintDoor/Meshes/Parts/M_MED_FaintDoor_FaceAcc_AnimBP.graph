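# Animation Blueprint for the FaintDoor male medium body's face accessory mesh, adding secondary hair motion on that player cosmetic.
# It copies the pose from the attached parent mesh, then runs a serial chain of single-bone Anim Dynamics nodes over the dyn_hair bones.
# Each node simulates bone-relative to root with angular springs, reduced gravity and per-bone angular limits, with a few bones blended at partial alpha.
# The event graph holds no logic.
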AnimBlueprint /BRCosmetics/Characters/Player/Male/Medium/Bodies/M_MED_FaintDoor/Meshes/Parts/M_MED_FaintDoor_FaceAcc_AnimBP
target_skeleton=&/M_MED_FaintDoor_FaceAcc_Skeleton
graphs [
 AnimGraph {nodes [
   n1 = Anim Dynamics - Bone: dyn_hair_01!(
    ComponentPose=Local To Component(LocalPose=Copy Pose From Mesh() {bUseAttachedParent=true})
    Alpha=0.7
    GravityOverride:Vector="0,0,-980"
    ExternalForce
   ) -> (Pose) {
    angularDampingOverride=1
    angularSpringConstant=1000
    bAngularSpring=true
    bOverrideAngularDamping=true
    gravityScale=0.35
    simulationSpace=BoneRelative
    alphaScaleBiasClamp {bInterpResult=true, interpSpeedDecreasing=5}
    boundBone {boneName=dyn_hair_01}
    componentAppliedLinearAccClamp=<3000 3000 3000>
    componentLinearAccScale=<0.3 0.3 0.3>
    componentLinearVelScale=<0.3 0.3 0.3>
    gravityOverride {z=-980}
    physicsBodyDefinitions [{
      boundBone {boneName=dyn_hair_01}
      boxExtents=<17 17 17>
      constraintSetup {twistAxis=AxisZ, angularLimitsMax=<0 15 0>, angularLimitsMin=<0 -15 -35>, angularTarget=<1 0 0>}
      localJointOffset=<5 1 -3>
     }]
    relativeSpaceBone {boneName=root}
    simSpaceSettings {simSpaceAngularAlpha=0.4}
   }
   n2 = Anim Dynamics - Bone: dyn_hair_02!(ComponentPose=@n1, Alpha=0.7, GravityOverride:Vector="0,0,-980", ExternalForce) -> (Pose) {
    angularDampingOverride=1
    angularSpringConstant=1000
    bAngularSpring=true
    bOverrideAngularDamping=true
    gravityScale=0.35
    simulationSpace=BoneRelative
    alphaScaleBiasClamp {bInterpResult=true, interpSpeedDecreasing=5}
    boundBone {boneName=dyn_hair_02}
    componentAppliedLinearAccClamp=<2000 2000 2000>
    componentLinearAccScale=<0.2 0.2 0.2>
    componentLinearVelScale=<0.3 0.3 0.3>
    gravityOverride {z=-980}
    physicsBodyDefinitions [{
      boundBone {boneName=dyn_hair_02}
      boxExtents=<18 18 18>
      constraintSetup {angularTargetAxis=AxisZ, angularLimitsMax=<0 15 0>, angularLimitsMin=<0 -15 -35>, angularTarget=<1 0 0>}
      localJointOffset=<3 0 -2>
     }]
    relativeSpaceBone {boneName=root}
    simSpaceSettings {simSpaceAngularAlpha=0.4}
   }
   n3 = Anim Dynamics - Bone: dyn_hair_03!(ComponentPose=@n2, Alpha=1, GravityOverride:Vector="0,0,-980", ExternalForce="0,0,0") -> (Pose) {
    angularDampingOverride=1
    angularSpringConstant=500
    bAngularSpring=true
    bOverrideAngularDamping=true
    gravityScale=0.35
    simulationSpace=BoneRelative
    alphaScaleBiasClamp {bInterpResult=true, interpSpeedDecreasing=5}
    boundBone {boneName=dyn_hair_03}
    componentAppliedLinearAccClamp=<3000 3000 3000>
    componentLinearAccScale=<0.2 0.2 0.2>
    componentLinearVelScale=<0.3 0.3 0.3>
    gravityOverride {z=-980}
    physicsBodyDefinitions [{
      boundBone {boneName=dyn_hair_03}
      boxExtents=<18 18 18>
      constraintSetup {twistAxis=AxisZ, angularLimitsMax=<0 5 15>, angularLimitsMin=<0 -35 -15>, angularTarget=<1 0 0>}
      localJointOffset=<6 0 -3>
     }]
    relativeSpaceBone {boneName=root}
    simSpaceSettings {simSpaceAngularAlpha=0.4}
   }
   n4 = Anim Dynamics - Bone: dyn_hair_04!(ComponentPose=@n3, Alpha=1, GravityOverride:Vector="0,0,-980", ExternalForce="0,0,0") -> (Pose) {
    angularDampingOverride=1
    angularSpringConstant=1000
    bAngularSpring=true
    bOverrideAngularDamping=true
    gravityScale=0.35
    simulationSpace=BoneRelative
    alphaScaleBiasClamp {bInterpResult=true, interpSpeedDecreasing=5}
    boundBone {boneName=dyn_hair_04}
    componentAppliedLinearAccClamp=<3000 3000 3000>
    componentLinearAccScale=<0.3 0.3 0.3>
    componentLinearVelScale=<0.3 0.3 0.3>
    gravityOverride {z=-980}
    physicsBodyDefinitions [{
      boundBone {boneName=dyn_hair_04}
      boxExtents=<15 15 15>
      constraintSetup {angularLimitsMax=<0 0 15>, angularLimitsMin=<0 -35 -15>, angularTarget=<1 0 0>}
      localJointOffset=<5 0 -3>
     }]
    relativeSpaceBone {boneName=root}
    simSpaceSettings {simSpaceAngularAlpha=0.4}
   }
   n5 = Anim Dynamics - Bone: dyn_hair_05!(ComponentPose=@n4, Alpha=1, GravityOverride:Vector="0,0,-980", ExternalForce="0,0,0") -> (Pose) {
    angularDampingOverride=1
    angularSpringConstant=1000
    bAngularSpring=true
    bOverrideAngularDamping=true
    gravityScale=0.35
    simulationSpace=BoneRelative
    alphaScaleBiasClamp {bInterpResult=true, interpSpeedDecreasing=5}
    boundBone {boneName=dyn_hair_05}
    componentAppliedLinearAccClamp=<3000 3000 3000>
    componentLinearAccScale=<0.3 0.3 0.3>
    componentLinearVelScale=<0.3 0.3 0.3>
    gravityOverride {z=-980}
    physicsBodyDefinitions [{
      boundBone {boneName=dyn_hair_05}
      boxExtents=<15 15 15>
      constraintSetup {angularLimitsMax=<0 0 15>, angularLimitsMin=<0 -35 -15>, angularTarget=<1 0 0>}
      localJointOffset=<4 -2 0>
     }]
    relativeSpaceBone {boneName=root}
    simSpaceSettings {simSpaceAngularAlpha=0.4}
   }
   n6 = Anim Dynamics - Bone: dyn_hair_06!(ComponentPose=@n5, Alpha=1, GravityOverride:Vector="0,0,-980", ExternalForce="0,0,0") -> (Pose) {
    angularDampingOverride=1
    angularSpringConstant=1000
    bAngularSpring=true
    bOverrideAngularDamping=true
    gravityScale=0.35
    simulationSpace=BoneRelative
    alphaScaleBiasClamp {bInterpResult=true, interpSpeedDecreasing=5}
    boundBone {boneName=dyn_hair_06}
    componentAppliedLinearAccClamp=<3000 3000 3000>
    componentLinearAccScale=<0.3 0.3 0.3>
    componentLinearVelScale=<0.3 0.3 0.3>
    gravityOverride {z=-980}
    physicsBodyDefinitions [{
      boundBone {boneName=dyn_hair_06}
      boxExtents=<13 13 13>
      constraintSetup {twistAxis=AxisZ, angularLimitsMax=<0 15 15>, angularLimitsMin=<0 -15 -15>, angularTarget=<1 0 0>}
      localJointOffset=<3 0 4>
     }]
    relativeSpaceBone {boneName=root}
    simSpaceSettings {simSpaceAngularAlpha=0.4}
   }
   n7 = Anim Dynamics - Bone: dyn_hair_07!(ComponentPose=@n6, Alpha=1, GravityOverride:Vector="0,0,-980", ExternalForce="0,0,0") -> (Pose) {
    angularDampingOverride=1
    angularSpringConstant=1000
    bAngularSpring=true
    bOverrideAngularDamping=true
    gravityScale=0.35
    simulationSpace=BoneRelative
    alphaScaleBiasClamp {bInterpResult=true, interpSpeedDecreasing=5}
    boundBone {boneName=dyn_hair_07}
    componentAppliedLinearAccClamp=<3000 3000 3000>
    componentLinearAccScale=<0.3 0.3 0.3>
    componentLinearVelScale=<0.3 0.3 0.3>
    gravityOverride {z=-980}
    physicsBodyDefinitions [{
      boundBone {boneName=dyn_hair_07}
      boxExtents=<15 15 15>
      constraintSetup {angularLimitsMax=<0 15 0>, angularLimitsMin=<0 -15 -35>, angularTarget=<1 0 0>}
      localJointOffset=<4 -2 0>
     }]
    relativeSpaceBone {boneName=root}
    simSpaceSettings {simSpaceAngularAlpha=0.4}
   }
   n8 = Anim Dynamics - Bone: dyn_hair_08!(ComponentPose=@n7, Alpha=1, GravityOverride:Vector="0,0,-980", ExternalForce="0,0,0") -> (Pose) {
    angularDampingOverride=1
    angularSpringConstant=1000
    bAngularSpring=true
    bOverrideAngularDamping=true
    gravityScale=0.35
    simulationSpace=BoneRelative
    alphaScaleBiasClamp {bInterpResult=true, interpSpeedDecreasing=5}
    boundBone {boneName=dyn_hair_08}
    componentAppliedLinearAccClamp=<3000 3000 3000>
    componentLinearAccScale=<0.3 0.3 0.3>
    componentLinearVelScale=<0.3 0.3 0.3>
    gravityOverride {z=-980}
    physicsBodyDefinitions [{
      boundBone {boneName=dyn_hair_08}
      boxExtents=<14 14 14>
      constraintSetup {twistAxis=AxisZ, angularLimitsMax=<0 20 0>, angularLimitsMin=<0 -20 -35>, angularTarget=<1 0 0>}
      localJointOffset=<4 0 4>
     }]
    relativeSpaceBone {boneName=root}
    simSpaceSettings {simSpaceAngularAlpha=0.4}
   }
   n9 = Anim Dynamics - Bone: dyn_hair_09!(ComponentPose=@n8, Alpha=0.5, GravityOverride:Vector="0,0,-980", ExternalForce="0,0,0") -> (Pose) {
    angularDampingOverride=1
    angularSpringConstant=1000
    bAngularSpring=true
    bOverrideAngularDamping=true
    gravityScale=0.35
    simulationSpace=BoneRelative
    alphaScaleBiasClamp {bInterpResult=true, interpSpeedDecreasing=5}
    boundBone {boneName=dyn_hair_09}
    componentAppliedLinearAccClamp=<3000 3000 3000>
    componentLinearAccScale=<0.3 0.3 0.3>
    componentLinearVelScale=<0.3 0.3 0.3>
    gravityOverride {z=-980}
    physicsBodyDefinitions [{
      boundBone {boneName=dyn_hair_09}
      boxExtents=<10 10 10>
      constraintSetup {angularLimitsMax=<0 15 0>, angularLimitsMin=<0 -15 -35>, angularTarget=<1 0 0>}
      localJointOffset=<5 0 0>
     }]
    relativeSpaceBone {boneName=root}
    simSpaceSettings {simSpaceAngularAlpha=0.4}
   }
   n10 = Anim Dynamics - Bone: dyn_hair_10!(ComponentPose=@n9, Alpha=1, GravityOverride:Vector="0,0,-980", ExternalForce="0,0,0") -> (Pose) {
    angularDampingOverride=1
    angularSpringConstant=1000
    bAngularSpring=true
    bOverrideAngularDamping=true
    gravityScale=0.35
    simulationSpace=BoneRelative
    alphaScaleBiasClamp {bInterpResult=true, interpSpeedDecreasing=5}
    boundBone {boneName=dyn_hair_10}
    componentAppliedLinearAccClamp=<3000 3000 3000>
    componentLinearAccScale=<0.3 0.3 0.3>
    componentLinearVelScale=<0.3 0.3 0.3>
    gravityOverride {z=-980}
    physicsBodyDefinitions [{
      boundBone {boneName=dyn_hair_10}
      boxExtents=<12 12 12>
      constraintSetup {angularLimitsMax=<0 15 0>, angularLimitsMin=<0 -15 -35>, angularTarget=<1 0 0>}
      localJointOffset=<5 0 0>
     }]
    relativeSpaceBone {boneName=root}
    simSpaceSettings {simSpaceAngularAlpha=0.4}
   }
   n11 = Anim Dynamics - Bone: dyn_hair_11!(ComponentPose=@n10, Alpha=1, GravityOverride:Vector="0,0,-980", ExternalForce="0,0,0") -> (Pose) {
    angularDampingOverride=1
    angularSpringConstant=1000
    bAngularSpring=true
    bOverrideAngularDamping=true
    gravityScale=0.35
    simulationSpace=BoneRelative
    alphaScaleBiasClamp {bInterpResult=true, interpSpeedDecreasing=5}
    boundBone {boneName=dyn_hair_11}
    componentAppliedLinearAccClamp=<3000 3000 3000>
    componentLinearAccScale=<0.3 0.3 0.3>
    componentLinearVelScale=<0.3 0.3 0.3>
    gravityOverride {z=-980}
    physicsBodyDefinitions [{
      boundBone {boneName=dyn_hair_11}
      boxExtents=<12 12 12>
      constraintSetup {angularLimitsMax=<0 20 0>, angularLimitsMin=<0 -20 -35>, angularTarget=<1 0 0>}
      localJointOffset=<5 0 -2>
     }]
    relativeSpaceBone {boneName=root}
    simSpaceSettings {simSpaceAngularAlpha=0.4}
   }
   n12 = Anim Dynamics - Bone: dyn_hair_12b!(ComponentPose=@n11, Alpha=1, GravityOverride:Vector="0,0,-980", ExternalForce="0,0,0") -> (Pose) {
    angularDampingOverride=1
    angularSpringConstant=5000
    bAngularSpring=true
    bOverrideAngularDamping=true
    gravityScale=0.35
    simulationSpace=BoneRelative
    alphaScaleBiasClamp {bInterpResult=true, interpSpeedDecreasing=5}
    boundBone {boneName=dyn_hair_12b}
    componentAppliedLinearAccClamp=<2000 2000 2000>
    componentLinearAccScale=<0.2 0.2 0.2>
    componentLinearVelScale=<0.2 0.2 0.2>
    gravityOverride {z=-980}
    physicsBodyDefinitions [{
      boundBone {boneName=dyn_hair_12b}
      boxExtents=<22 22 22>
      constraintSetup {angularLimitsMax=<0 10 0>, angularLimitsMin=<0 -10 -20>, angularTarget=<1 0 0>}
      localJointOffset=<1 0 -2>
     }]
    relativeSpaceBone {boneName=root}
    simSpaceSettings {simSpaceAngularAlpha=0.4}
   }
   n13 = Anim Dynamics - Bone: dyn_hair_12!(ComponentPose=@n12, Alpha=1, GravityOverride:Vector="0,0,-980", ExternalForce="0,0,0") -> (Pose) {
    angularDampingOverride=1
    angularSpringConstant=5000
    bAngularSpring=true
    bOverrideAngularDamping=true
    gravityScale=0.35
    simulationSpace=BoneRelative
    alphaScaleBiasClamp {bInterpResult=true, interpSpeedDecreasing=5}
    boundBone {boneName=dyn_hair_12}
    componentAppliedLinearAccClamp=<2000 2000 2000>
    componentLinearAccScale=<0.2 0.2 0.2>
    componentLinearVelScale=<0.2 0.2 0.2>
    gravityOverride {z=-980}
    physicsBodyDefinitions [{
      boundBone {boneName=dyn_hair_12}
      boxExtents=<22 22 22>
      constraintSetup {angularLimitsMax=<0 10 0>, angularLimitsMin=<0 -10 -20>, angularTarget=<1 0 0>}
      localJointOffset=<1 0 0>
     }]
    relativeSpaceBone {boneName=root}
    simSpaceSettings {simSpaceAngularAlpha=0.4}
   }
   n14 = Anim Dynamics - Bone: dyn_hair_13!(ComponentPose=@n13, Alpha=1, GravityOverride:Vector="0,0,-980", ExternalForce="0,0,0") -> (Pose) {
    angularDampingOverride=1
    angularSpringConstant=1000
    bAngularSpring=true
    bOverrideAngularDamping=true
    gravityScale=0.35
    simulationSpace=BoneRelative
    alphaScaleBiasClamp {bInterpResult=true, interpSpeedDecreasing=5}
    boundBone {boneName=dyn_hair_13}
    componentAppliedLinearAccClamp=<3000 3000 3000>
    componentLinearAccScale=<0.3 0.3 0.3>
    componentLinearVelScale=<0.3 0.3 0.3>
    gravityOverride {z=-980}
    physicsBodyDefinitions [{
      boundBone {boneName=dyn_hair_13}
      boxExtents=<15 15 15>
      constraintSetup {twistAxis=AxisZ, angularLimitsMax=<0 15 0>, angularLimitsMin=<0 -15 -35>, angularTarget=<1 0 0>}
      localJointOffset=<3 2 0>
     }]
    relativeSpaceBone {boneName=root}
    simSpaceSettings {simSpaceAngularAlpha=0.4}
   }
   n15 = Anim Dynamics - Bone: dyn_hair_14!(ComponentPose=@n14, Alpha=1, GravityOverride:Vector="0,0,-980", ExternalForce="0,0,0") -> (Pose) {
    angularDampingOverride=1
    angularSpringConstant=1000
    bAngularSpring=true
    bOverrideAngularDamping=true
    gravityScale=0.35
    simulationSpace=BoneRelative
    alphaScaleBiasClamp {bInterpResult=true, interpSpeedDecreasing=5}
    boundBone {boneName=dyn_hair_14}
    componentAppliedLinearAccClamp=<3000 3000 3000>
    componentLinearAccScale=<0.3 0.3 0.3>
    componentLinearVelScale=<0.3 0.3 0.3>
    gravityOverride {z=-980}
    physicsBodyDefinitions [{
      boundBone {boneName=dyn_hair_14}
      boxExtents=<15 15 15>
      constraintSetup {angularLimitsMax=<0 15 0>, angularLimitsMin=<0 -15 -35>, angularTarget=<1 0 0>}
      localJointOffset=<5 0 -2>
     }]
    relativeSpaceBone {boneName=root}
    simSpaceSettings {simSpaceAngularAlpha=0.4}
   }
   n16 = Anim Dynamics - Bone: dyn_hair_15!(ComponentPose=@n15, Alpha=1, GravityOverride:Vector="0,0,-980", ExternalForce="0,0,0") -> (Pose) {
    angularDampingOverride=1
    angularSpringConstant=1000
    bAngularSpring=true
    bOverrideAngularDamping=true
    gravityScale=0.35
    simulationSpace=BoneRelative
    alphaScaleBiasClamp {bInterpResult=true, interpSpeedDecreasing=5}
    boundBone {boneName=dyn_hair_15}
    componentAppliedLinearAccClamp=<3000 3000 3000>
    componentLinearAccScale=<0.3 0.3 0.3>
    componentLinearVelScale=<0.3 0.3 0.3>
    gravityOverride {z=-980}
    physicsBodyDefinitions [{
      boundBone {boneName=dyn_hair_15}
      boxExtents=<10 10 10>
      constraintSetup {angularLimitsMax=<0 20 0>, angularLimitsMin=<0 -20 -35>, angularTarget=<1 0 0>}
      localJointOffset=<5 0 -3>
     }]
    relativeSpaceBone {boneName=root}
    simSpaceSettings {simSpaceAngularAlpha=0.4}
   }
   n17 = Anim Dynamics - Bone: dyn_hair_16!(ComponentPose=@n16, Alpha=1, GravityOverride:Vector="0,0,-980", ExternalForce="0,0,0") -> (Pose) {
    angularDampingOverride=1
    angularSpringConstant=1000
    bAngularSpring=true
    bOverrideAngularDamping=true
    gravityScale=0.35
    simulationSpace=BoneRelative
    alphaScaleBiasClamp {bInterpResult=true, interpSpeedDecreasing=5}
    boundBone {boneName=dyn_hair_16}
    componentAppliedLinearAccClamp=<3000 3000 3000>
    componentLinearAccScale=<0.3 0.3 0.3>
    componentLinearVelScale=<0.3 0.3 0.3>
    gravityOverride {z=-980}
    physicsBodyDefinitions [{
      boundBone {boneName=dyn_hair_16}
      boxExtents=<15 15 15>
      constraintSetup {angularLimitsMax=<0 25 0>, angularLimitsMin=<0 -25 -35>, angularTarget=<1 0 0>}
      localJointOffset=<5 0 -2>
     }]
    relativeSpaceBone {boneName=root}
    simSpaceSettings {simSpaceAngularAlpha=0.4}
   }
   n18 = Anim Dynamics - Bone: dyn_hair_17!(ComponentPose=@n17, Alpha=1, GravityOverride:Vector="0,0,-980", ExternalForce="0,0,0") -> (Pose) {
    angularDampingOverride=1
    angularSpringConstant=1000
    bAngularSpring=true
    bOverrideAngularDamping=true
    gravityScale=0.35
    simulationSpace=BoneRelative
    alphaScaleBiasClamp {bInterpResult=true, interpSpeedDecreasing=5}
    boundBone {boneName=dyn_hair_17}
    componentAppliedLinearAccClamp=<3000 3000 3000>
    componentLinearAccScale=<0.3 0.3 0.3>
    componentLinearVelScale=<0.3 0.3 0.3>
    gravityOverride {z=-980}
    physicsBodyDefinitions [{
      boundBone {boneName=dyn_hair_17}
      boxExtents=<15 15 15>
      constraintSetup {angularLimitsMax=<0 25 0>, angularLimitsMin=<0 -25 -35>, angularTarget=<1 0 0>}
      localJointOffset=<5 -2 0>
     }]
    relativeSpaceBone {boneName=root}
    simSpaceSettings {simSpaceAngularAlpha=0.4}
   }
   Output Pose(Result=Component To Local(ComponentPose=@n18)) {layerGroup=None}
  ]}
 EventGraph {nodes [Event BlueprintUpdateAnimation()]}
]
variables []
